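; Creative device that places an Overlord Spire boss with weakpoints and configurable beam, homing projectile, slam and scream attacks.
; User options feed runtime values with Verse accessors, using per-actor data registry scopes to scale values per device.
; Minigame events start and stop StateTree logic, while team/class-filtered sensed players drive targeting, music and activation events.
; Replicated state handles spawning, death, collision and map icon; a nav invoker supplies navmesh for attack queries.
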(event OnComponentBeginOverlap(VisualizationRangeTrigger)
  (params (OverlappedComponent :type Primitive%20Component%20Object%20Reference) (OtherActor :type Actor%20Object%20Reference) (OtherComp :type Primitive%20Component%20Object%20Reference) (OtherBodyIndex :type Integer) (bFromSweep :type Boolean) (SweepResult :type Hit%20Result%20Structure%20%28by%20ref%29))
  ; Visuals
  (if (Math|Boolean|ANDBoolean (HelperFunctions|IsLocalPawn OtherActor) (Variables|UserOptions|Basic|GetShowVisualizationRange) (not (Building|IsPreviewBuildingActor)))
    (CallFunction|ShowRangeVisual)))

(event EventBeginPlay
  ; Initialization and Properties
  (|CallParentFunction)
  (DataRegistry|InitializeDataRegistryResolverScopes) ; Setup for per-actor scoped in the data registry
  (Initialization|GatherWeakpointActors)
  (CallFunction|TryRegisterNavInvoker)
  (CallFunction|DeactivateMusic))

(event EventEndPlay
  (params (EndPlayReason :type EEndPlayReason%20Enum :default "Destroyed"))
  ; Initialization and Properties
  (MapIcon|ClearMapIcon))

(event Custom|TryRegisterNavInvoker
  (bind _self self)
  (bind _returnvalue (Fortnite|Athena|GameState|GetGameStateAthenafromContext _self))
  ; Initialization and Properties
  ;;[0.5156 0.4131 0.1864 0.5] The device requires nav mesh to be built in order to do EQS queries and navigation projection for attack abilities
  (Creative|Networking|Authority/Cosmetic
    (:Authority
      (if (Variables|/Script/FortniteGame.FortGameStateAthena|GetbPlaylistDataIsLoaded _returnvalue)
        (Class|BPCreativeCommonAILibrary|TryAddNavInvokeronDevice self "/CRD_OverlordSpire/Components/OverlordSpireDeviceNavInvokerClass.OverlordSpireDeviceNavInvokerClass_C") ; Will add Nav Invoker around this device if Nav Mesh on Demand is enabled
        (else
          (Playlist|BindEventtoOnCurrentPlaylistLoaded _returnvalue (delegate Custom|OnPlaylistLoaded)))))))

(event OnComponentEndOverlap(VisualizationRangeTrigger)
  (params (OverlappedComponent :type Primitive%20Component%20Object%20Reference) (OtherActor :type Actor%20Object%20Reference) (OtherComp :type Primitive%20Component%20Object%20Reference) (OtherBodyIndex :type Integer))
  ; Visuals
  (if (Math|Boolean|ANDBoolean (HelperFunctions|IsLocalPawn OtherActor) (Variables|UserOptions|Basic|GetShowVisualizationRange) (not (Building|IsPreviewBuildingActor)))
    (CallFunction|HideRangeVisual false)))

(event Custom|OnPlaylistLoaded
  (params (PlaylistName :type Name) (PlaylistContextTags :type Gameplay%20Tag%20Container%20Structure%20%28by%20ref%29))
  ; Initialization and Properties
  ;;[0.5156 0.4131 0.1864 0.5] The device requires nav mesh to be built in order to do EQS queries and navigation projection for attack abilities
  (Playlist|UnbindEventfromOnCurrentPlaylistLoaded (Fortnite|Athena|GameState|GetGameStateAthenafromContext self) (delegate Custom|OnPlaylistLoaded))
  (Class|BPCreativeCommonAILibrary|TryAddNavInvokeronDevice self "/CRD_OverlordSpire/Components/OverlordSpireDeviceNavInvokerClass.OverlordSpireDeviceNavInvokerClass_C") ; Will add Nav Invoker around this device if Nav Mesh on Demand is enabled
)

(event Custom|ShowRangeVisual
  ; Visuals
  (Utilities|IsValid (Variables|/Script/Engine.StaticMeshComponent|GetStaticMesh (Variables|Default|GetActivationRangeVisualization))
    (:Is%20Valid
      (Development|SetHiddeninGame (Variables|Default|GetActivationRangeVisualization) false false)
      (bind _material_instance_dynamic (Visualization|GetVisualizationSphereMID))
      (bind (_direction _lerp) (Timeline|Timeline Play)
        (:Update
          (Rendering|Material|SetScalarParameterValue (stale Custom|HideRangeVisual _material_instance_dynamic) "Opacity" _lerp :!class /Script/Engine.MaterialInstanceDynamic)
          (Transformation|SetRelativeScale3D (Variables|Default|GetActivationRangeVisualization) (Math|Conversions|ToVector(Float) (Math|Float|Lerp 0.0 (/ (* (Variables|UserOptions|Basic|GetActivation%20Distance) 1e+02) (Variables|Visualization|GetVisualRangeMeshBoundsRadius)) _lerp))))
        (:Finished
          (switch Utilities|FlowControl|Switch|SwitchonETimelineDirection _direction
            (:Backward
              (Development|SetHiddeninGame (Variables|Default|GetActivationRangeVisualization) true false))))))))

(event OnAnyPropertyChanged(ToyOptionsComponent)
  ; Initialization and Properties
  (Initialization|Initialize (Variables|Default|GetCreative_UseRestriction_Component) (Variables|Default|GetFortMinigameProgress) 0 (Variables|Default|GetToyOptionsComponent) (Variables|UserOptions|Advanced|GetTargetTeam) (Variables|UserOptions|Advanced|GetInvertTargetTeam) (Variables|UserOptions|Advanced|GetTargetClass) (Variables|UserOptions|Advanced|GetInvertTargetClass) false)
  (Initialization|InitRuntimeValues)
  (CallFunction|UpdateDataRegistryValues)
  (Visualization|UpdateVisualizationState)
  (CallFunction|SetUpHealthThresholds)
  (Networking|SwitchHasAuthority
    (:Authority
      (CallFunction|DelayedSetMaxHealth) ; Delay reading data registry values since the registry has a 1s cache timer
      (if (Variables|RuntimeValues|GetbShowMapIconRuntime)
        (MapIcon|ShowMapIcon)
        (else
          (MapIcon|ClearMapIcon))))
    (:Remote
      (if (Variables|UserOptions|Advanced|GetbPlayMusic)
        (CallFunction|ActivateMusic)
        (else
          (CallFunction|DeactivateMusic))))))

(event OnPropertiesLoaded(ToyOptionsComponent)
  ; Initialization and Properties
  (Initialization|Initialize (Variables|Default|GetCreative_UseRestriction_Component) (Variables|Default|GetFortMinigameProgress) 0 (Variables|Default|GetToyOptionsComponent) (Variables|UserOptions|Advanced|GetTargetTeam) (Variables|UserOptions|Advanced|GetInvertTargetTeam) (Variables|UserOptions|Advanced|GetTargetClass) (Variables|UserOptions|Advanced|GetInvertTargetClass) false)
  (Initialization|InitRuntimeValues)
  (CallFunction|UpdateDataRegistryValues)
  (Visualization|UpdateVisualizationState)
  (CallFunction|SetUpHealthThresholds)
  (Networking|SwitchHasAuthority
    (:Authority
      (CallFunction|DelayedSetMaxHealth) ; Delay reading data registry values since the registry has a 1s cache timer
      (if (Variables|RuntimeValues|GetbShowMapIconRuntime)
        (MapIcon|ShowMapIcon)
        (else
          (MapIcon|ClearMapIcon))))
    (:Remote
      (if (Variables|UserOptions|Advanced|GetbPlayMusic)
        (CallFunction|ActivateMusic)
        (else
          (CallFunction|DeactivateMusic))))))

(event Custom|OnPlayModeChangedDelegate_Event
  (params (Minigame :type Fort%20Minigame%20Object%20Reference) (bIsInPlayMode :type Boolean)))

(event Custom|HideRangeVisual
  (params (Immediate :type Boolean))
  ; Visuals
  (Utilities|IsValid (Variables|/Script/Engine.StaticMeshComponent|GetStaticMesh (Variables|Default|GetActivationRangeVisualization))
    (:Is%20Valid
      (bind _material_instance_dynamic (Visualization|GetVisualizationSphereMID))
      (if (not Immediate)
        (Timeline|Timeline Reverse)
        (else
          (Timeline|Timeline SetNewTime))))))

(event OnMeshandMaterialsLoaded(ActivationRangeVisualization)
  (bind (_origin _boxextent _sphereradius) (Math|BoxSphereBounds|BreakBoxSphereBounds (StaticMesh|GetBounds (Variables|/Script/Engine.StaticMeshComponent|GetStaticMesh (Variables|Default|GetActivationRangeVisualization)))))
  ; Visuals
  (Variables|Visualization|SetVisualRangeMeshBoundsRadius _sphereradius)
  (Visualization|UpdateVisualizationState))

(event MinigameStartedDelegate(FortMinigameProgress)
  (bind _fortminigameprogress (Variables|Default|GetFortMinigameProgress))
  (bind _currentminigame (Variables|/Script/FortniteGame.FortMinigameLogicComponent|GetCurrentMinigame _fortminigameprogress))
  ; Minigame Events
  (Initialization|InitializeDefaultState)
  (CallFunction|ReloadUserOptions)
  (if (and (Variables|UserOptions|Advanced|GetbPlayMusic) (Variables|State|GetbSpawned))
    (CallFunction|ActivateMusic))
  (Networking|CanExecuteCosmeticEvents
    (:True
      (if (Variables|UserOptions|Basic|GetbShowPreviewClouds)
        (CallFunction|SpawnPreviewClouds)))
    (:False
      _))
  (Networking|SwitchHasAuthority
    (:Authority
      (StateFlow|StartStateLogic)
      (Utilities|IsValid _currentminigame
        (:Is%20Valid
          (Minigame|BindEventtoPlayerChangedTeamDelegate _currentminigame (EventDispatchers|CreateEvent :!event OnPlayerChangedTeams))
          (Minigame|BindEventtoOnPlayerClassApplied _currentminigame (EventDispatchers|CreateEvent :!event OnPlayerChangedClass)))))))

(event MinigameEndedDelegate(FortMinigameProgress)
  (bind _fortminigameprogress (Variables|Default|GetFortMinigameProgress))
  (bind _currentminigame (Variables|/Script/FortniteGame.FortMinigameLogicComponent|GetCurrentMinigame _fortminigameprogress))
  ; Minigame Events
  (StateFlow|StopStateLogic)
  (CallFunction|DeactivateMusic)
  (StateFlow|SetEnabledState false)
  (Class|SpireBossAnimBP|SetIsDead (Variables|Default|GetSpireBoss_Anim%20BP) false)
  (Networking|CanExecuteCosmeticEvents
    (:True
      (CallFunction|DestroyClouds))
    (:False
      _))
  (Networking|SwitchHasAuthority
    (:Authority
      (Variables|State|SetbSpawned true) ; Spire should be spawned when returning to edit mode
      (Utilities|IsValid _currentminigame
        (:Is%20Valid
          (Minigame|UnbindEventfromPlayerChangedTeamDelegate _currentminigame (EventDispatchers|CreateEvent :!event OnPlayerChangedTeams))
          (Minigame|UnbindEventfromOnPlayerClassApplied _currentminigame (EventDispatchers|CreateEvent :!event OnPlayerChangedClass)))))))

(event Custom|SpawnPreviewClouds
  (bind (_returnvalue_location _returnvalue_rotation) (Math|Transform|BreakTransform (Transformation|GetActorTransform)))
  ; Visuals
  (if (not (Utilities|IsValid (Variables|Clouds|GetCloudsActor) :!class /Script/Engine.KismetSystemLibrary))
    (bind _class (Utilities|AsyncLoadClassAsset (Variables|Clouds|GetCloudsClass))
      (:Completed
        (bind _ascrd_low_altitude_clouds_overlord_spire_device (Utilities|Casting|CastToCRD_LowAltitudeClouds_OverlordSpireDeviceClass _class))
        (bind _spawnedactor (Fortnite|Utilities|SpawnTransientActor _ascrd_low_altitude_clouds_overlord_spire_device (Math|Transform|MakeTransform (+ _returnvalue_location (* (Transformation|GetActorUpVector) (Variables|Clouds|GetCloudRelativeLocation))) _returnvalue_rotation) "AlwaysSpawn" self))
        (Variables|Clouds|SetCloudsActor _spawnedactor))) ; Async laod is fine since clouds won't appear in UEFN
  ))

(event OnAddedtoMinigame(FortMinigameProgress)
  (params (Minigame :type Fort%20Minigame%20Object%20Reference))
  ; Minigame Events
  (Minigame|AssignOnPlayModeChangedDelegate Minigame (delegate Custom|OnPlayModeChanged))
  (Visualization|UpdateVisualizationState)
  (if (not (Minigame|IsinPlayMode Minigame :!class /Script/FortniteGame.FortMinigame))
    (Networking|SwitchHasAuthority
      (:Authority
        (Variables|State|SetbSpawned true) ; Spire should be spawned when in edit mode
      )
      (:Remote
        _))
    (for _array_element (Variables|Default|GetWeakpointActors)
      (Class|BGAWeakpointActorBase|UpdateWeakpointCollision _array_element false (Variables|/ActorWeakpointsCore/GameplayActors/BGA_WeakpointActor_Base.BGA_WeakpointActor_Base_C|GetWeakpointCollisionKind _array_element) 0))
    (Fortnite|SetCanbeDamaged self)))

(event Custom|OnPlayModeChanged
  (params (Minigame :type Fort%20Minigame%20Object%20Reference) (bIsInPlayMode :type Boolean))
  ; Minigame Events
  (Visualization|UpdateVisualizationState)
  (if (not bIsInPlayMode)
    (Networking|SwitchHasAuthority
      (:Authority
        (Variables|State|SetbSpawned true) ; Spire should be spawned when in edit mode
      )
      (:Remote
        _))
    (for _array_element (Variables|Default|GetWeakpointActors)
      (Class|BGAWeakpointActorBase|UpdateWeakpointCollision _array_element false (Variables|/ActorWeakpointsCore/GameplayActors/BGA_WeakpointActor_Base.BGA_WeakpointActor_Base_C|GetWeakpointCollisionKind _array_element) 0))
    (Fortnite|SetCanbeDamaged self)))

(event OnRemovedfromMinigame(FortMinigameProgress)
  (params (Minigame :type Fort%20Minigame%20Object%20Reference))
  ; Minigame Events
  (Minigame|UnbindEventfromOnPlayModeChangedDelegate Minigame (EventDispatchers|CreateEvent :!event Custom|OnPlayModeChanged)))

(event EventOnStateTransitionedEvent
  (params (PreviousState :type Gameplay%20Tag%20Structure%20%28by%20ref%29) (NewState :type Gameplay%20Tag%20Structure%20%28by%20ref%29))
  ; State Flow
  (|CallParentFunction PreviousState NewState)
  (switch Utilities|FlowControl|Switch|SwitchonGameplayTag NewState
    (:Spire.Major.State.Active.Idle
      _)
    (:Spire.Major.State.Active.Combat
      (AI|Perception|RegisterwithPerceptionSystem (Variables|Default|GetAthenaAIPerceptionStimuliSource)))
    (:Spire.Major.State.Despawned
      (Variables|ForCreative|SetbFinishedIntroSequence false) ; Clear intro sequence flag so that it can be retriggered once spawned again
    ))
  (Variables|STData|SetbForceSpawn false) ; Clear ForceSpawn flag once in a spawned state
)

(event Custom|HandleReset
  ; State Flow
  (if (or (Variables|State|GetbSpawned) (Variables|State|GetbSpireDestroyed))
    (Class|SpireBossAnimBP|SetIsDead (Variables|Default|GetSpireBoss_Anim%20BP) false)
    (Networking|SwitchHasAuthority
      (:Authority
        (Gameplay|StateTree|SendStateTreeEvent (Variables|Default|GetStateTree) (Utilities|Struct|MakeStateTreeEvent "(TagName=\"Event.SB.Reset\")") :!class /Script/GameplayStateTreeModule.StateTreeComponent)
        (if (and (Variables|UserOptions|Advanced|GetbPlayMusic) (Variables|UserOptions|Basic|GetbSpawnedAtGameStart))
          (CallFunction|ActivateMusic)
          (else
            (CallFunction|DeactivateMusic))))
      (:Remote
        _))
    (if (Variables|State|GetbSpireDestroyed)
      (Initialization|InitializeDefaultState)
      (Networking|SwitchHasAuthority
        (:Authority
          (StateFlow|StartStateLogic)))
      (else
        (Initialization|InitializeDefaultState) ; calling in both true/false since it can't be called before checking if spire was destroyed since it overwrites it, and I dont want  to make a member variable to cache it
      ))))

(event EventOnDamageServer
  (params (Damage :type Float%20%28single-precision%29) (DamageTags :type Gameplay%20Tag%20Container%20Structure%20%28by%20ref%29) (Momentum :type Vector :default "0, 0, 0") (HitInfo :type Hit%20Result%20Structure%20%28by%20ref%29) (InstigatedBy :type Controller%20Object%20Reference) (DamageCauser :type Actor%20Object%20Reference) (EffectContext :type Gameplay%20Effect%20Context%20Handle%20Structure))
  (|CallParentFunction Damage DamageTags Momentum HitInfo InstigatedBy DamageCauser EffectContext)
  (if (not (Health|HasHealthLeft))
    (CallFunction|SetHealthtoMax)
    (Destruction|HandleOnDeathServer InstigatedBy)))

(event Custom|SetAbilitySFValue
  (params (SF_Row :type Scalable%20Float%20Structure) (NewValue :type Float%20%28double-precision%29))
  (ScalableFloat|SetValuewithContext SF_Row NewValue))

(fn AddSensedTarget
  (params (Object :type Object%20Reference))
  (bind _asfort_player_pawn (Utilities|Casting|CastToFortPlayerPawn Object))
  (if (and (StateChecks|IsValidTeam (Variables|Default|GetCreative_UseRestriction_Component) _asfort_player_pawn) (StateChecks|IsValidClass (Variables|Default|GetCreative_UseRestriction_Component) _asfort_player_pawn))
    (PawnTracking|AddSensedTargetInternal _asfort_player_pawn)
    (if (== (Variables|Default|GetOverrideTarget) _asfort_player_pawn)
      (CallFunction|UpdateCurrentTarget))))

(fn AddSensedTarget_Internal
  (params (SensedTarget :type Fort%20Player%20Pawn%20Object%20Reference%20%28by%20ref%29))
  (bind _returnvalue (Utilities|Array|AddUnique (Variables|Default|GetSensedTargets) SensedTarget))
  (if (>= _returnvalue 0)
    (Variables|Default|SetSensedTargets (Variables|Default|GetSensedTargets))
    (return true)
    (else
      (return false))))

(fn BlueprintCanInteract
  (params (InteractingPawn :type Fort%20Pawn%20Object%20Reference) (InteractionBeingAttempted :type EInteractionBeingAttempted%20Enum :default "FirstInteraction") (InteractionType :type TInteractionType%20Enum :default "IT_NoInteraction"))
  (return (Interaction|CanEditOptionsComponent (Variables|Default|GetToyOptionsComponent) (Variables|GetInteractingPawn))))

(fn BlueprintGetInteractionString
  (params (InteractingPawn :type Fort%20Pawn%20Object%20Reference) (InteractionBeingAttempted :type EInteractionBeingAttempted%20Enum :default "FirstInteraction"))
  (return "NSLOCTEXT(\"[0A6365A67918485AC7304F3EE5B7FDD7]\", \"21B37A574BFCEC32A17E129792F415C7\", \"Customize\")"))

(fn BuildDataRegistryResolverScope
  (return (Variables|Default|GetResolver%20Scopes) 1 true))

(fn BuildExposedAssetProviderVersion(Implementation)
  (params (Version :type Fort%20Exposed%20Asset%20Provider%20Version%20Structure%20%28by%20ref%29))
  (ExposedAssetProvider|AddVersion Version 1)
  (return))

(fn CanPerformBeamAttack
  (return (Variables|RuntimeValues|GetbCanPerformBeamAttackRuntime)))

(fn CanPerformHomingProjectileAttack
  (return (Variables|RuntimeValues|GetbCanPerformHomingProjectileRuntime)))

(fn CanPerformScreamAbility
  (return (Variables|RuntimeValues|GetbCanPerformScreamAttackRuntime)))

(fn CanPerformSlamAttack
  (return (Variables|RuntimeValues|GetbCanPerformSlamAttackRuntime)))

(fn ClearMapIcon
  (bind _mapiconactor (Variables|MapIcon|GetMapIconActor)
    (:then
      (Actor|DestroyActor _mapiconactor))
    (:else)))

(fn CollectClassReferencestoAllow(Implementation)
  (params (Collector :type Fort%20Exposed%20Class%20Reference%20Collector%20Structure%20%28by%20ref%29))
  (ExposedAssetProvider|CollectObjectPropertyClassReferences Collector 0 (Variables|Default|GetBP_MusicEncounterComponent))
  (return))

(fn DelayedSetMaxHealth
  (Utilities|Time|ClearandInvalidateTimerbyHandle (Variables|Default|GetDelayedSetMaxHealthTimer))
  (bind _returnvalue (Utilities|Time|SetTimerbyEvent (EventDispatchers|CreateEvent :!event SetScaledHealth) 1.0)) ; 1s since that's the data registry's cache timer
  (Variables|Default|SetDelayedSetMaxHealthTimer _returnvalue))

(fn DespawnAfterDeathAnim
  (StateFlow|SetSpireHidden true)
  (MapIcon|ClearMapIcon))

(fn DestroyClouds
  (bind _cloudsactor (Variables|Clouds|GetCloudsActor)
    (:then
      (Actor|DestroyActor _cloudsactor))
    (:else)))

(fn DisableSpireCollision
  (Collision|SetCollisionEnabled (Variables|Default|GetSkeletalMeshComponentBudgeted) "NoCollision")
  (for _array_element (Variables|Default|GetWeakpointActors)
    (Class|BGAWeakpointActorBase|UpdateWeakpointCollision _array_element false (Variables|/ActorWeakpointsCore/GameplayActors/BGA_WeakpointActor_Base.BGA_WeakpointActor_Base_C|GetWeakpointCollisionKind _array_element) 0)))

(fn EnableSpireCollision
  (Collision|SetCollisionEnabled (Variables|Default|GetCapsule) "QueryAndPhysics")
  (Collision|SetCollisionEnabled (Variables|Default|GetSkeletalMeshComponentBudgeted) "QueryOnly")
  (for _array_element (Variables|Default|GetWeakpointActors)
    (Class|BGAWeakpointActorBase|UpdateWeakpointCollision _array_element (Minigame|IsinProgress (Minigame|GetMinigame self :!class /Script/FortniteGame.MinigameBlueprintLibrary)) (Variables|/ActorWeakpointsCore/GameplayActors/BGA_WeakpointActor_Base.BGA_WeakpointActor_Base_C|GetWeakpointCollisionKind _array_element) 0)))

(fn GatherWeakpointActors
  (Networking|SwitchHasAuthority
    (:Authority
      (LocalVariables|SetWeakpointChildActorComponents (Utilities|Array|MakeArray (Variables|Default|GetWeakpoint01) (Variables|Default|GetWeakpoint02) (Variables|Default|GetWeakpoint03) (Variables|Default|GetWeakpoint04)))
      (for _index (range 3)
        (bind _ascrd_weakpoint_actor_overlord_spire (Utilities|Casting|CastToCRD_WeakpointActor_OverlordSpire (Variables|/Script/Engine.ChildActorComponent|GetChildActor (Utilities|Array|Get(acopy) (LocalVariables|GetWeakpointChildActorComponents) _index))))
        (Utilities|Array|Add (LocalVariables|GetTempWeakpointActors) _ascrd_weakpoint_actor_overlord_spire)
        (Class|CRDWeakpointActorOverlordSpire|SetSpireOwner _ascrd_weakpoint_actor_overlord_spire self))
      (bind _ascrd_weakpoint_actor_core_overlord_spire (Utilities|Casting|CastToCRD_WeakpointActor_Core_OverlordSpire (Variables|/Script/Engine.ChildActorComponent|GetChildActor (Utilities|Array|Get(acopy) (LocalVariables|GetWeakpointChildActorComponents) 3)))
        (:then
          (Utilities|Array|Add (LocalVariables|GetTempWeakpointActors) _ascrd_weakpoint_actor_core_overlord_spire)
          (Default|SetSpireOwner _ascrd_weakpoint_actor_core_overlord_spire self))
        (:CastFailed
          _))
      (Variables|Default|SetWeakpointActors (LocalVariables|GetTempWeakpointActors)) ; Replicating array instead of constructing on the client due to actor replication order/timing
    )))

(fn GetActivationDistanceVerse
  (return (Variables|RuntimeValues|GetActivationDistanceRuntime)))

(fn GetBeamArcVerse
  (return (Variables|Verse|Abilities|Beam|GetBeamArcVerse)))

(fn GetBeamAttackCooldown
  (return (Variables|RuntimeValues|GetBeamCooldownTimeRuntime)))

(fn GetBeamDamagePerSecondToBuildingsVerse
  (return (Variables|Verse|Abilities|Beam|GetBeamDamageToBuildingsVerse)))

(fn GetBeamDamagePerSecondToCreaturesVerse
  (return (Variables|Verse|Abilities|Beam|GetBeamDamageToCreaturesVerse)))

(fn GetBeamDamagePerSecondToGuardsVerse
  (return (Variables|Verse|Abilities|Beam|GetBeamDamageToGuardsVerse)))

(fn GetBeamDamagePerSecondToPlayersVerse
  (return (Variables|Verse|Abilities|Beam|GetBeamDamageToPlayersVerse)))

(fn GetBeamDamagePerSecondToVehiclesVerse
  (return (Variables|Verse|Abilities|Beam|GetBeamDamageToVehiclesVerse)))

(fn GetBeamDamagePerSecondToWildlifeVerse
  (return (Variables|Verse|Abilities|Beam|GetBeamDamageToWildlifeVerse)))

(fn GetBeamDurationVerse
  (return (Variables|Verse|Abilities|Beam|GetBeamDurationVerse)))

(fn GetCacheContextOverride
  (return 0))

(fn GetCollisionStaticMesh
  (bind _returnvalue_min (Utilities|Struct|BreakBox (StaticMesh|GetBoundingBox (Variables|Default|GetSpireStaticMesh))))
  (bind (_returnvalue_min_x _returnvalue_min_y _returnvalue_min_z) (Math|Vector|BreakVector _returnvalue_min))
  (return 0.0 0.0 (* -1.0 _returnvalue_min_z) (Variables|Default|GetSpireStaticMesh)))

(fn GetFirstControllerFromActors
  (params (InActors :type Array%20of%20Actor%20Object%20References))
  (for _array_element InActors
    (Utilities|IsValid _array_element
      (:Is%20Valid
        (Utilities|IsValid (Fortnite|GetControllerfromActor _array_element)
          (:Is%20Valid
            (LocalVariables|SetFoundController (Fortnite|GetControllerfromActor _array_element)))))))
  (return (LocalVariables|GetFoundController)))

(fn GetHealthVerse
  (return (Health|GetHealth)))

(fn GetHomingProjectileCooldown
  (return (Variables|RuntimeValues|GetHomingProjectileCooldownTimeRuntime)))

(fn GetHomingProjectileDamageCreaturesVerse
  (return (Variables|Verse|Abilities|HomProj|GetHomingProjectileDamageToCreaturesVerse)))

(fn GetHomingProjectileDamageToBuildings
  (return (Variables|Verse|Abilities|HomProj|GetHomingProjectileDamageToBuildingsVerse)))

(fn GetHomingProjectileDamageToGuardsVerse
  (return (Variables|Verse|Abilities|HomProj|GetHomingProjectileDamageToGuardsVerse)))

(fn GetHomingProjectileDamageToPlayersVerse
  (return (Variables|Verse|Abilities|HomProj|GetHomingProjectileDamageToPlayersVerse)))

(fn GetHomingProjectileDamageToVehiclesVerse
  (return (Variables|Verse|Abilities|HomProj|GetHomingProjectileDamageToVehiclesVerse)))

(fn GetHomingProjectileDamageWildlifeVerse
  (return (Variables|Verse|Abilities|HomProj|GetHomingProjectileDamageToWildlifeVerse)))

(fn GetHomingProjectileMaxSpeedVerse
  (return (Variables|Verse|Abilities|HomProj|GetHomingProjectileMaxSpeedVerse)))

(fn GetMaxHealthVerse
  (return (Health|GetMaxHealth)))

(fn GetMaxTimeBetweenAbilities
  (return (Variables|RuntimeValues|GetMaximumTimeBetweenAbilitiesRuntime)))

(fn GetMeshComponents
  (return (Utilities|Array|MakeArray (Variables|Default|GetSkeletalMeshComponentBudgeted))))

(fn GetMinTimeBetweenAbilities
  (return (Variables|RuntimeValues|GetMinimumTimeBetweenAbilitiesRuntime)))

(fn GetNumberOfHomingProjectilesVerse
  (return (Variables|Verse|Abilities|HomProj|GetNumberOfHomingProjectilesVerse)))

(fn GetSensedTargets
  (return (Variables|Default|GetSensedTargets)))

(fn GetShowMapIconVerse
  (return (Variables|RuntimeValues|GetbShowMapIconRuntime)))

(fn GetSlamAttackCooldown
  (return (Variables|RuntimeValues|GetSlamCooldownTimeRuntime)))

(fn GetSlamAttackDamageCreaturesVerse
  (return (Variables|Verse|Abilities|Slam|GetSlamDamageToCreaturesVerse)))

(fn GetSlamAttackDamageToBuildingsVerse
  (return (Variables|Verse|Abilities|Slam|GetSlamDamageToBuildingsVerse)))

(fn GetSlamAttackDamageToGuardsVerse
  (return (Variables|Verse|Abilities|Slam|GetSlamDamageToGuardsVerse)))

(fn GetSlamAttackDamageToPlayersVerse
  (return (Variables|Verse|Abilities|Slam|GetSlamDamageToPlayersVerse)))

(fn GetSlamAttackDamageToVehiclesVerse
  (return (Variables|Verse|Abilities|Slam|GetSlamDamageToVehiclesVerse)))

(fn GetSlamAttackDamageWildlifeVerse
  (return (Variables|Verse|Abilities|Slam|GetSlamDamageToWildlifeVerse)))

(fn GetSlamInitialRadiusVerse
  (return (Variables|Verse|Abilities|Slam|GetSlamInitialRadiusVerse)))

(fn GetSlamRecoveryTimeVerse
  (return (Variables|Verse|Abilities|Slam|GetSlamRecoveryTimeVerse)))

(fn GetSlamShockwaveDurationVerse
  (return (Variables|Verse|Abilities|Slam|GetSlamShockwaveDurationVerse)))

(fn GetSlamShockwaveMaxRadiusVerse
  (return (Variables|Verse|Abilities|Slam|GetSlamShockwaveMaxRadiusVerse)))

(fn GetVisualizationSphereMID
  (bind _asmaterial_instance_dynamic (Utilities|Casting|CastToMaterialInstanceDynamic (Rendering|Material|GetMaterial (Variables|Default|GetActivationRangeVisualization) 0))
    (:then
      (return _asmaterial_instance_dynamic))
    (:CastFailed
      (return (Material|GetMID (Variables|Default|GetActivationRangeVisualization))))))

(fn GetWaitForAllProjectilesToSpawnBeforeNextAttackVerse
  (return (Variables|RuntimeValues|GetbWaitForAllProjectilesToSpawnBeforeNextAttackRuntime)))

(fn HandleDeactivated
  (MapIcon|UpdateMapIconActiveState false)
  (AI|Perception|UnregisterfromPerceptionSystem (Variables|Default|GetAthenaAIPerceptionStimuliSource))
  (GameplayEvent|BroadcastEvent (Variables|UserOptions|Events|GetOnDeactivate) 0 (Utilities|Struct|MakeGameplayEventonSpireDeactivated 0)))

(fn HandleOnDeathServer
  (params (InstigatorController :type Controller%20Object%20Reference))
  (|CallParentFunction InstigatorController)
  (AI|Perception|UnregisterfromPerceptionSystem (Variables|Default|GetAthenaAIPerceptionStimuliSource))
  (StateFlow|StopStateLogic)
  (StateFlow|SetEnabledState false)
  (Fortnite|SetCanbeDamaged self)
  (Variables|State|SetbSpireDestroyed true)
  (if (Minigame|IsinProgress (Minigame|GetMinigame self :!class /Script/FortniteGame.MinigameBlueprintLibrary))
    (GameplayEvent|BroadcastEvent (Variables|UserOptions|Events|GetOnDestroy) 0 (Utilities|Struct|MakeGameplayEventonSpireDestroyed (Variables|GetInstigatorController)))))

(fn HandlePawnAffiliationChanged
  (params (PlayerState :type Fort%20Player%20State%20Object%20Reference))
  (Utilities|IsValid PlayerState
    (:Is%20Valid
      (bind _asfort_player_pawn (Utilities|Casting|CastToFortPlayerPawn (PlayerState|GetPawn PlayerState)))
      (if (Utilities|Array|ContainsItem (Variables|InternalData|GetSensedPlayerPawns) _asfort_player_pawn)
        (if (and (StateChecks|IsValidTeam (Variables|Default|GetCreative_UseRestriction_Component) _asfort_player_pawn) (StateChecks|IsValidClass (Variables|Default|GetCreative_UseRestriction_Component) _asfort_player_pawn))
          (bind _badded (PawnTracking|AddSensedTargetInternal _asfort_player_pawn))
          (bind _output_get (LocalVariables|SetbAddedTarget _badded))
          (else
            (bind _bremoved (PawnTracking|RemoveSensedTargetInternal _asfort_player_pawn))
            (bind _output_get_2 (LocalVariables|SetbRemovedTarget _bremoved))))
        (if (or _output_get _output_get_2)
          (Variables|InternalData|SetSensedPlayerPawnCount (Utilities|Array|Length (Variables|Default|GetSensedTargets)))
          (if (Math|Boolean|ORBoolean (and (LocalVariables|GetbAddedTarget) (not (Utilities|IsValid (Variables|InternalData|GetCurrentTargetPlayerPawn) :!class /Script/Engine.KismetSystemLibrary))) (and (LocalVariables|GetbRemovedTarget) (== _asfort_player_pawn (Variables|InternalData|GetCurrentTargetPlayerPawn))) (== _asfort_player_pawn (Variables|Default|GetOverrideTarget)))
            (CallFunction|UpdateCurrentTarget)
            (if (Utilities|Array|IsEmpty (Variables|Default|GetSensedTargets))
              (CallFunction|DeactivateMusic) ; If there are no more targets then the spire should stop the encounter music as ti will be in the idle state instead of combat
              (else
                (CallFunction|ActivateMusic)))))))))

(fn HandleSpawnedStatus
  (if (Variables|State|GetbSpawned)
    (StateFlow|RevealSpire)
    (StateFlow|EnableSpireCollision)
    (else
      (StateFlow|HideSpire)
      (StateFlow|DisableSpireCollision))))

(fn HandleWeakpointsOnSpireDestroyed
  (StateFlow|DisableSpireCollision))

(fn HideSpire
  (Fortnite|SetCanbeDamaged self)
  (Rendering|SetActorHiddenInGame true)
  (for _array_element (Variables|Default|GetWeakpointActors)
    (Rendering|SetActorHiddenInGame _array_element true)))

(fn IgnoreWorldGridScaleOverride
  (return false))

(fn InitializeDataRegistryResolverScopes
  (Networking|SwitchHasAuthority
    (:Authority
      (if (Utilities|Array|IsEmpty (Variables|Default|GetResolver%20Scopes))
        (Variables|Default|SetResolver%20Scopes (Utilities|Array|MakeArray (SpatialGameplay|GetSpatialGameplayInterfaceFNameforActor self) (Utilities|String|StringToName (Utilities|GetDisplayName self))))))))

(fn InitializeDefaultState
  (|CallParentFunction)
  (if (Minigame|IsinProgress (Minigame|GetMinigame self :!class /Script/FortniteGame.MinigameBlueprintLibrary))
    (StateFlow|SetEnabledState (Variables|UserOptions|Basic|GetbEnabledAtGameStart))
    (if (Variables|State|GetbSpawned)
      (StateFlow|HandleSpawnedStatus))
    (Networking|SwitchHasAuthority
      (:Authority
        (AI|Perception|UnregisterfromPerceptionSystem (Variables|Default|GetAthenaAIPerceptionStimuliSource)) ; Keep unregistered until the spire enters combat state (can be damaged)
        (Utilities|Time|ClearandInvalidateTimerbyHandle (Variables|Default|GetDespawnAfterDeathTimer))
        (CallFunction|SetScaledHealth)
        (CallFunction|UpdateMaterialHealth 1.0) ; Reset health-affected materials and VFX
        (Class|BPFLAICommonFunctions|RemoveAllInstancesOfGameplayTag self (Variables|InternalData|GetExplodingTag) true)
        (Variables|State|SetbSpawned (Variables|UserOptions|Basic|GetbSpawnedAtGameStart))))
    (else
      (StateFlow|SetEnabledState false) ; Disable perception (and consequently UI) in edit mode
    )))

(fn InitRuntimeValues
  (CallFunction|SetActivationDistanceInternal (* (Variables|UserOptions|Basic|GetActivation%20Distance) 1e+02))
  (Verse|General|SetShowMapIconVerse (Variables|UserOptions|Basic|GetbShowMapIcon))
  (Verse|Abilities|SetMinimumTimeBetweenAbilitiesVerse (Variables|UserOptions|Advanced|GetMinimumTimeBetweenAbilities))
  (Verse|Abilities|SetMaximumTimeBetweenAbilitiesVerse (Variables|UserOptions|Advanced|GetMaximumTimeBetweenAbilities))
  (Verse|Abilities|Beam|SetBeamAttackEnabledVerse (Variables|UserOptions|Basic|GetbCanPerformBeamAttack))
  (Verse|Abilities|Beam|SetBeamCooldownTimeVerse (Variables|UserOptions|Advanced|GetBeamCooldownTime))
  (Verse|Abilities|HomProj|SetHomingProjectileAttackEnabledVerse (Variables|UserOptions|Basic|GetbCanPerformHomingProjectile))
  (Verse|Abilities|HomProj|SetHomingProjectileCooldownTimeVerse (Variables|UserOptions|Advanced|GetHomingProjectileCooldownTime))
  (Verse|Abilities|Slam|SetSlamAttackEnabledVerse (Variables|UserOptions|Basic|GetbCanPerformSlamAttack))
  (Verse|Abilities|Slam|SetSlamCooldownTimeVerse (Variables|UserOptions|Advanced|GetSlamCooldownTime))
  (Verse|Abilities|HomProj|SetWaitforAllProjectilestoSpawnBeforeNextAttackVerse (Variables|UserOptions|Advanced|GetbWaitForAllProjectilesToSpawnBeforeNextAttack))
  (Verse|Abilities|SetScreamAttackEnabledVerse (Variables|UserOptions|Basic|GetbCanPerformScreamAttack)))

(fn InitSpireVariables
  (bind _bspawnedatgamestart (Variables|UserOptions|Basic|GetbSpawnedAtGameStart))
  (|CallParentFunction)
  (Variables|State|SetbSpireDestroyed false)
  (Variables|STData|SetbForceSpawn false)
  (Variables|WeakpointData|SetbFrontLeftAppendageActive _bspawnedatgamestart)
  (Variables|WeakpointData|SetbFrontRightAppendageActive _bspawnedatgamestart)
  (Variables|WeakpointData|SetbBackAppendageActive _bspawnedatgamestart)
  (Variables|WeakpointData|SetbSpireCoreActive _bspawnedatgamestart)
  (Variables|ForCreative|SetbFinishedIntroSequence _bspawnedatgamestart))

(fn IsActivatedVerse
  (return (and (not (Utilities|Array|IsEmpty (Variables|Default|GetSensedTargets))) (Variables|State|GetbEnabled))))

(fn IsDestroyedVerse
  (return (Variables|State|GetbSpireDestroyed)))

(fn IsEnabledVerse
  (return (Variables|State|GetbEnabled)))

(fn IsSpawnedVerse
  (return (Variables|State|GetbSpawned)))

(fn NotifyBeamAttackEvent
  (params (bEventStarted :type Boolean))
  (if bEventStarted
    (GameplayEvent|BroadcastEvent (Variables|UserOptions|Events|GetOnBeginBeamAttack) 0 (Utilities|Struct|MakeGameplayEventonSpireBeginBeamAttack 0))
    (else
      (GameplayEvent|BroadcastEvent (Variables|UserOptions|Events|GetOnEndBeamAttack) 0 (Utilities|Struct|MakeGameplayEventonSpireEndBeamAttack 0)))))

(fn NotifyHomingProjectileAttackEvent
  (params (bEventStarted :type Boolean))
  (if bEventStarted
    (GameplayEvent|BroadcastEvent (Variables|UserOptions|Events|GetOnBeginHomingProjectileAttack) 0 (Utilities|Struct|MakeGameplayEventonSpireBeginHomingProjectileAttack 0))
    (else
      (GameplayEvent|BroadcastEvent (Variables|UserOptions|Events|GetOnEndHomingProjectileAttack) 0 (Utilities|Struct|MakeGameplayEventonSpireEndHomingProjectileAttack 0)))))

(fn NotifyScreamAttackEvent
  (params (bEventStarted :type Boolean))
  (if bEventStarted
    (GameplayEvent|BroadcastEvent (Variables|UserOptions|Events|GetOnBeginScream) 0 (Utilities|Struct|MakeGameplayEventonSpireBeginScream 0))
    (else
      (GameplayEvent|BroadcastEvent (Variables|UserOptions|Events|GetOnEndScream) 0 (Utilities|Struct|MakeGameplayEventonSpireEndScream 0)))))

(fn NotifySlamAttackEvent
  (params (bEventStarted :type Boolean))
  (if bEventStarted
    (GameplayEvent|BroadcastEvent (Variables|UserOptions|Events|GetOnBeginSlamAttack) 0 (Utilities|Struct|MakeGameplayEventonSpireBeginSlamAttack 0))
    (else
      (GameplayEvent|BroadcastEvent (Variables|UserOptions|Events|GetOnEndSlamAttack) 0 (Utilities|Struct|MakeGameplayEventonSpireEndSlamAttack 0)))))

(fn NotifySlamRecoverEvent
  (params (bEventStarted :type Boolean))
  (if bEventStarted
    (GameplayEvent|BroadcastEvent (Variables|UserOptions|Events|GetOnBeginSlamRecover) 0 (Utilities|Struct|MakeGameplayEventonSpireBeginSlamRecover 0))
    (else
      (GameplayEvent|BroadcastEvent (Variables|UserOptions|Events|GetOnEndSlamRecover) 0 (Utilities|Struct|MakeGameplayEventonSpireEndSlamRecover 0)))))

(fn OnNearbyActorDetected
  (params (AllFoundActors :type Array%20of%20Actor%20Object%20References) (EnteringActors :type Array%20of%20Actor%20Object%20References) (ExitingActors :type Array%20of%20Actor%20Object%20References))
  (bind _enteringactors (Variables|GetEnteringActors))
  (bind _controller (CallFunction|GetFirstControllerfromActors _enteringactors))
  (Utilities|FlowControl|Sequence
    (:then_0
      (LocalVariables|SetbPrevDeactivated (Utilities|Array|IsEmpty (Variables|Default|GetSensedTargets)))
      (for _array_element (Variables|GetEnteringActors)
        (PawnTracking|AddSensedTarget _array_element))
      (for _array_element_2 (Variables|GetExitingActors)
        (PawnTracking|RemoveSensedTarget _array_element_2)))
    (:then_1
      (|CallParentFunction AllFoundActors EnteringActors ExitingActors)
      (if (Math|Boolean|ANDBoolean (LocalVariables|GetbPrevDeactivated) (not (Utilities|Array|IsEmpty (Variables|Default|GetSensedTargets))) (Variables|State|GetbEnabled))
        (MapIcon|UpdateMapIconActiveState true)
        (Utilities|IsValid _controller
          (:Is%20Valid
            (GameplayEvent|BroadcastEvent (Variables|UserOptions|Events|GetOnActivate) 0 (Utilities|Struct|MakeGameplayEventonSpireActivated _controller))))
        (elif (Utilities|Array|IsEmpty (Variables|Default|GetSensedTargets))
          (CallFunction|HandleDeactivated))))))

(fn OnPlayerChangedClass
  (params (PlayerState :type Fort%20Player%20State%20Object%20Reference) (ClassSlotIndex :type Byte))
  (PawnTracking|HandlePawnAffiliationChanged PlayerState))

(fn OnPlayerChangedTeams
  (params (PlayerState :type Fort%20Player%20State%20Object%20Reference) (TeamId :type Byte) (OldTeamId :type Byte) (bTeamChangedWithoutRespawn :type Boolean))
  (PawnTracking|HandlePawnAffiliationChanged PlayerState))

(fn OnRep_bPlayMusic
  (if (Variables|UserOptions|Advanced|GetbPlayMusic)
    (CallFunction|ActivateMusic)
    (else
      (CallFunction|DeactivateMusic))))

(fn OnRep_bSpawned
  (StateFlow|HandleSpawnedStatus)
  (Networking|SwitchHasAuthority
    (:Authority
      (if (Variables|State|GetbSpawned)
        (if (Variables|RuntimeValues|GetbShowMapIconRuntime)
          (MapIcon|ShowMapIcon))
        (if (Minigame|IsinProgress (Minigame|GetMinigame self :!class /Script/FortniteGame.MinigameBlueprintLibrary))
          (GameplayEvent|BroadcastEvent (Variables|UserOptions|Events|GetOnSpawn) 0 (Utilities|Struct|MakeGameplayEventonSpireSpawned 0)))
        (else
          (MapIcon|ClearMapIcon))))))

(fn OnRep_bSpireDestroyed
  (if (Variables|State|GetbSpireDestroyed)
    (Networking|SwitchHasAuthority
      (:Authority
        ; Set timer to clean up VFX and hide meshes (which would normally be handled by the actor being destroyed at the end of its lifespan)
        (bind _returnvalue (Utilities|Time|SetTimerbyEvent (EventDispatchers|CreateEvent :!event DespawnAfterDeathAnim) (Variables|Default|GetDespawnAfterDeathAnimTime)))
        (Variables|Default|SetDespawnAfterDeathTimer _returnvalue))
      (:Remote
        (StateFlow|DisableSpireCollision)))))

(fn OnRep_SensedTargets
  (Networking|SwitchHasAuthority
    (:Remote
      (bind _asfort_player_pawn (Utilities|Casting|CastToFortPlayerPawn (Pawn|GetControlledPawn (Game|GetPlayerController 0))))
      (bind _playerstate (Variables|/Script/Engine.Pawn|GetPlayerState _asfort_player_pawn)
        (:then
          (Utilities|IsValid (Actor|GetComponentbyClass _playerstate "/SpiresGameplay/Gameplay/Components/BP_SpireEventPlayerComponent.BP_SpireEventPlayerComponent_C")
            (:Is%20Valid
              (if (Utilities|Array|ContainsItem (Variables|Default|GetSensedTargets) _asfort_player_pawn)
                ;[1 0 0.01402 0.5] TODO: Prevent this from triggering multiple times for the same player
                (Class|SpireEventPlayerComponent|StartParticipatinginSpireEvent (Actor|GetComponentbyClass _playerstate "/SpiresGameplay/Gameplay/Components/BP_SpireEventPlayerComponent.BP_SpireEventPlayerComponent_C") self)
                (else
                  (Class|SpireEventPlayerComponent|StopParticipatinginSpireEvent (Actor|GetComponentbyClass _playerstate "/SpiresGameplay/Gameplay/Components/BP_SpireEventPlayerComponent.BP_SpireEventPlayerComponent_C") self))))))
        (:else)))))

(fn OnRep_WeakpointActors
  (Networking|SwitchHasAuthority
    (:Remote
      (if (not (Minigame|IsinPlayMode self :!class /Script/FortniteGame.MinigameBlueprintLibrary))
        (for _array_element (Variables|Default|GetWeakpointActors)
          (Utilities|IsValid _array_element
            (:Is%20Not%20Valid
              (return))))
        (for _array_element_2 (Variables|Default|GetWeakpointActors)
          (Class|BGAWeakpointActorBase|UpdateWeakpointCollision _array_element_2 false (Variables|/ActorWeakpointsCore/GameplayActors/BGA_WeakpointActor_Base.BGA_WeakpointActor_Base_C|GetWeakpointCollisionKind _array_element_2) 0))
        (Fortnite|SetCanbeDamaged self)))))

(fn RegisterPerception
  (bind _returnvalue (Query|BPRegisterListener (Variables|Default|GetFortNearbyActorsPerception) (Utilities|Struct|MakeFortNearbyActorsPerceptionConfiguration (ScalableFloat|GetValueatLevel (Variables|InternalData|GetSF_DetectionUpdateRate)) 3 0 (Variables|RuntimeValues|GetActivationDistanceRuntime) (Utilities|Array|MakeArray "Hostile")) (EventDispatchers|CreateEvent :!event OnNearbyActorDetected)))
  (Variables|ForCreative|SetPerceptionListenerHandle _returnvalue))

(fn ReloadUserOptions
  (Initialization|InitRuntimeValues)
  (CallFunction|UpdateDataRegistryValues))

(fn RemoveSensedTarget
  (params (Object :type Object%20Reference))
  (bind _asfort_player_pawn (Utilities|Casting|CastToFortPlayerPawn Object))
  (PawnTracking|RemoveSensedTargetInternal _asfort_player_pawn)
  (if (and (== _asfort_player_pawn (Variables|Default|GetOverrideTarget)) (== (Variables|Default|GetOverrideTarget) (Variables|InternalData|GetCurrentTargetPlayerPawn)))
    (CallFunction|UpdateCurrentTarget)))

(fn RemoveSensedTarget_Internal
  (params (SensedTarget :type Fort%20Player%20Pawn%20Object%20Reference%20%28by%20ref%29))
  (bind _returnvalue (Utilities|Array|RemoveItem (Variables|Default|GetSensedTargets) SensedTarget))
  (if _returnvalue
    (Variables|Default|SetSensedTargets (Variables|Default|GetSensedTargets))
    (return true)
    (else
      (return false))))

(fn RevealSpire
  (Fortnite|SetCanbeDamaged self (Minigame|IsinProgress (Minigame|GetMinigame self :!class /Script/FortniteGame.MinigameBlueprintLibrary)))
  (Rendering|SetActorHiddenInGame false)
  (for _array_element (Variables|Default|GetWeakpointActors)
    (Rendering|SetActorHiddenInGame _array_element false)))

(fn SetActivationDistance_Internal
  (params (Value :type Float%20%28double-precision%29))
  (Variables|RuntimeValues|SetActivationDistanceRuntime (Math|Float|Clamp(Float) Value 5e+02 1e+04))
  (Music|UpdateMusicRadius))

(fn SetActivationDistanceVerse
  (params (NewActivationDistance :type Float%20%28double-precision%29))
  (CallFunction|SetActivationDistanceInternal NewActivationDistance)
  (Music|UpdateMusicRadius)
  (Initialization|UnregisterPerception)
  (Initialization|RegisterPerception))

(fn SetBeamArcVerse
  (params (Arc :type Float%20%28double-precision%29))
  (bind _output_get (Variables|Verse|Abilities|Beam|SetBeamArcVerse (Math|Float|Clamp(Float) Arc 0.0 3.6e+02)))
  (ScalableFloat|SetValuewithContext (Variables|BeamData|GetSF_BeamYawStart) (* (/ _output_get 2.0) -1.0))
  (ScalableFloat|SetValuewithContext (Variables|BeamData|GetSF_BeamYawEnd) (/ _output_get 2.0)))

(fn SetBeamAttackEnabledVerse
  (params (bEnabled :type Boolean))
  (Variables|RuntimeValues|SetbCanPerformBeamAttackRuntime bEnabled))

(fn SetBeamCooldownTimeVerse
  (params (Cooldown :type Float%20%28double-precision%29))
  (Variables|RuntimeValues|SetBeamCooldownTimeRuntime (Math|Float|Clamp(Float) Cooldown 1.0 6e+01)))

(fn SetBeamDamagePerSecondToBuildingsVerse
  (params (Damage :type Float%20%28double-precision%29))
  (bind _output_get (Variables|Verse|Abilities|Beam|SetBeamDamageToBuildingsVerse (Math|Float|Clamp(Float) Damage 0.0 1e+03)))
  (ScalableFloat|SetValuewithContext (Variables|ScalableFloats|GetSF_BeamDmg_Building) _output_get)
  (ScalableFloat|SetValuewithContext (Variables|ScalableFloats|GetSF_BeamDmg_BuildPlayer) _output_get))

(fn SetBeamDamagePerSecondToCreaturesVerse
  (params (Damage :type Float%20%28double-precision%29))
  (bind _output_get (Variables|Verse|Abilities|Beam|SetBeamDamageToCreaturesVerse (Math|Float|Clamp(Float) Damage 0.0 1e+03)))
  (ScalableFloat|SetValuewithContext (Variables|ScalableFloats|GetSF_BeamDmg_Creature) _output_get))

(fn SetBeamDamagePerSecondToGuardsVerse
  (params (Damage :type Float%20%28double-precision%29))
  (bind _output_get (Variables|Verse|Abilities|Beam|SetBeamDamageToGuardsVerse (Math|Float|Clamp(Float) Damage 0.0 1e+03)))
  (ScalableFloat|SetValuewithContext (Variables|ScalableFloats|GetSF_BeamDmg_Guard) _output_get))

(fn SetBeamDamagePerSecondToPlayersVerse
  (params (Damage :type Float%20%28double-precision%29))
  (bind _output_get (Variables|Verse|Abilities|Beam|SetBeamDamageToPlayersVerse (Math|Float|Clamp(Float) Damage 0.0 1e+03)))
  (ScalableFloat|SetValuewithContext (Variables|ScalableFloats|GetSF_BeamDmg_Player) _output_get))

(fn SetBeamDamagePerSecondToVehiclesVerse
  (params (Damage :type Float%20%28double-precision%29))
  (bind _output_get (Variables|Verse|Abilities|Beam|SetBeamDamageToVehiclesVerse (Math|Float|Clamp(Float) Damage 0.0 1e+03)))
  (ScalableFloat|SetValuewithContext (Variables|ScalableFloats|GetSF_BeamDmg_Vehicle) _output_get))

(fn SetBeamDamagePerSecondToWildlifeVerse
  (params (Damage :type Float%20%28double-precision%29))
  (bind _output_get (Variables|Verse|Abilities|Beam|SetBeamDamageToWildlifeVerse (Math|Float|Clamp(Float) Damage 0.0 1e+03)))
  (ScalableFloat|SetValuewithContext (Variables|ScalableFloats|GetSF_BeamDmg_Wildlife) _output_get))

(fn SetBeamDurationVerse
  (params (Duration :type Float%20%28double-precision%29))
  (bind _output_get (Variables|Verse|Abilities|Beam|SetBeamDurationVerse (Math|Float|Clamp(Float) Duration 1.0 6e+01)))
  (CallFunction|SetAbilitySFValue (Variables|BeamData|GetSF_BeamFireTime) _output_get))

(fn SetCurrentTargetPlayer
  (params (NewTarget :type Fort%20Player%20Pawn%20Object%20Reference))
  (LocalVariables|SetbSettingNewTarget (!= NewTarget (Variables|InternalData|GetCurrentTargetPlayerPawn)))
  (|CallParentFunction NewTarget)
  (if (LocalVariables|GetbSettingNewTarget)
    (GameplayEvent|BroadcastEvent (Variables|UserOptions|Events|GetOnTargetChanged) 0 (Utilities|Struct|MakeGameplayEventonSpireTargetUpdated (Pawn|GetController (Variables|InternalData|GetCurrentTargetPlayerPawn))))))

(fn SetEnabledState
  (params (bNewEnabled :type Boolean))
  (if (!= bNewEnabled (Variables|State|GetbEnabled))
    (Variables|State|SetbEnabled bNewEnabled)
    (Networking|SwitchHasAuthority
      (:Authority
        (if bNewEnabled
          (Initialization|RegisterPerception)
          (else
            (Initialization|UnregisterPerception)))))))

(fn SetHealthToMax
  (Variables|Default|SetCurrentHealthThreshold 0)
  (Health|SetHealth (Health|GetMaxHealth))
  (for _array_element (Variables|Default|GetWeakpointActors)
    (Class|BGAWeakpointActorBase|ResetWeakpointHealth _array_element)))

(fn SetHealthVerse
  (params (NewHealth :type Float%20%28double-precision%29))
  (bind _self self)
  (bind _returnvalue (Health|GetHealthPercent _self))
  (bind _returnvalue_2 (< _returnvalue 0.66))
  (bind _returnvalue_3 (< _returnvalue 0.33))
  (LocalVariables|SetPrevHealth (Health|GetHealth))
  (LocalVariables|SetPrevHealthThreshold (Variables|Default|GetCurrentHealthThreshold))
  (Health|SetHealth (Math|Float|Clamp(Float) (Variables|GetNewHealth) 0.0 (Health|GetMaxHealth)))
  (if (> (Variables|GetNewHealth) 0)
    (Variables|InternalData|SetbFirstHealthThresholdReached _returnvalue_2)
    (Variables|InternalData|SetbSecondHealthThresholdReached _returnvalue_3)
    (CallFunction|UpdateMaterialHealth (Math|Float|SelectFloat 1.0 (Math|Float|SelectFloat 0.5 0.0 _returnvalue_3) _returnvalue_2))
    (Variables|Default|SetCurrentHealthThreshold 0) ; Reset to 0 so next function updates it properly
    (CallFunction|CheckforCrossedHealthThreshold)
    (if (< (Variables|Default|GetCurrentHealthThreshold) (LocalVariables|GetPrevHealthThreshold))
      (Default|CallOnSpireDamageThresholdReached)
      (Abilities|TryActivateAbilitybyClass (Ability|GetAbilitySystemComponent 0) (Variables|Default|GetFullBodyHitReactAbility) true))
    (else
      (Destruction|HandleOnDeathServer 0))))

(fn SetHomingProjectileAttackEnabledVerse
  (params (bEnabled :type Boolean))
  (Variables|RuntimeValues|SetbCanPerformHomingProjectileRuntime bEnabled))

(fn SetHomingProjectileCooldownTimeVerse
  (params (Cooldown :type Float%20%28double-precision%29))
  (Variables|RuntimeValues|SetHomingProjectileCooldownTimeRuntime (Math|Float|Clamp(Float) Cooldown 1.0 6e+01)))

(fn SetHomingProjectileDamageCreaturesVerse
  (params (Damage :type Float%20%28double-precision%29))
  (bind _output_get (Variables|Verse|Abilities|HomProj|SetHomingProjectileDamageToCreaturesVerse (Math|Float|Clamp(Float) Damage 0.0 1e+04)))
  (ScalableFloat|SetValuewithContext (Variables|ScalableFloats|GetSF_ProjectileDmg_Creature) _output_get))

(fn SetHomingProjectileDamageToBuildingsVerse
  (params (Damage :type Float%20%28double-precision%29))
  (bind _output_get (Variables|Verse|Abilities|HomProj|SetHomingProjectileDamageToBuildingsVerse (Math|Float|Clamp(Float) Damage 0.0 1e+04)))
  (ScalableFloat|SetValuewithContext (Variables|ScalableFloats|GetSF_ProjectileDmg_Building) _output_get)
  (ScalableFloat|SetValuewithContext (Variables|ScalableFloats|GetSF_ProjectileDmg_BuildingPlayer) _output_get))

(fn SetHomingProjectileDamageToGuardsVerse
  (params (Damage :type Float%20%28double-precision%29))
  (bind _output_get (Variables|Verse|Abilities|HomProj|SetHomingProjectileDamageToGuardsVerse (Math|Float|Clamp(Float) Damage 0.0 1e+04)))
  (ScalableFloat|SetValuewithContext (Variables|ScalableFloats|GetSF_ProjectileDmg_Guard) _output_get))

(fn SetHomingProjectileDamageToPlayersVerse
  (params (Damage :type Float%20%28double-precision%29))
  (bind _output_get (Variables|Verse|Abilities|HomProj|SetHomingProjectileDamageToPlayersVerse (Math|Float|Clamp(Float) Damage 0.0 1e+04)))
  (ScalableFloat|SetValuewithContext (Variables|ScalableFloats|GetSF_ProjectileDmg_Player) _output_get))

(fn SetHomingProjectileDamageToVehiclesVerse
  (params (Damage :type Float%20%28double-precision%29))
  (bind _output_get (Variables|Verse|Abilities|HomProj|SetHomingProjectileDamageToVehiclesVerse (Math|Float|Clamp(Float) Damage 0.0 1e+04)))
  (ScalableFloat|SetValuewithContext (Variables|ScalableFloats|GetSF_ProjectileDmg_Vehicle) _output_get))

(fn SetHomingProjectileDamageWildlifeVerse
  (params (Damage :type Float%20%28double-precision%29))
  (bind _output_get (Variables|Verse|Abilities|HomProj|SetHomingProjectileDamageToWildlifeVerse (Math|Float|Clamp(Float) Damage 0.0 1e+04)))
  (ScalableFloat|SetValuewithContext (Variables|ScalableFloats|GetSF_ProjectileDmg_Wildlife) _output_get))

(fn SetHomingProjectileMaxSpeedVerse
  (params (MaxSpeed :type Float%20%28double-precision%29))
  (bind _output_get (Variables|Verse|Abilities|HomProj|SetHomingProjectileMaxSpeedVerse (Math|Float|Clamp(Float) MaxSpeed 1e+02 1e+04)))
  (ScalableFloat|SetValuewithContext (Variables|ScalableFloats|GetSF_HomingProjectileMaxSpeed) _output_get))

(fn SetMaximumTimeBetweenAbilitiesVerse
  (params (MaxTimeBetweenAbilities :type Float%20%28double-precision%29))
  (Variables|RuntimeValues|SetMaximumTimeBetweenAbilitiesRuntime (Math|Float|Clamp(Float) MaxTimeBetweenAbilities 1.0 6e+01)))

(fn SetMinimumTimeBetweenAbilitiesVerse
  (params (MinTimeBetweenAbilities :type Float%20%28double-precision%29))
  (Variables|RuntimeValues|SetMinimumTimeBetweenAbilitiesRuntime (Math|Float|Clamp(Float) MinTimeBetweenAbilities 1.0 6e+01)))

(fn SetNumberOfHomingProjectilesVerse
  (params (NumProjectiles :type Integer))
  (bind _output_get (Variables|Verse|Abilities|HomProj|SetNumberOfHomingProjectilesVerse (Math|Integer|Clamp(Integer) NumProjectiles 1 50)))
  (ScalableFloat|SetValuewithContext (Variables|ScalableFloats|GetSF_NumHomingProjectiles) _output_get))

(fn SetScaledHealth
  (Health|SetMaxHealthDontCreateASC (Variables|UserOptions|Basic|GetMaximum%20Health))
  (CallFunction|SetHealthtoMax)
  (Variables|InternalData|SetbHealthInitialized true))

(fn SetScreamAttackEnabledVerse
  (params (bEnabled :type Boolean))
  (Variables|RuntimeValues|SetbCanPerformScreamAttackRuntime bEnabled))

(fn SetShowMapIconVerse
  (params (bNewShowMapIcon :type Boolean))
  (bind _output_get (Variables|RuntimeValues|SetbShowMapIconRuntime bNewShowMapIcon))
  (if _output_get
    (MapIcon|ShowMapIcon)
    (else
      (MapIcon|ClearMapIcon))))

(fn SetSlamAttackDamageCreaturesVerse
  (params (Damage :type Float%20%28double-precision%29))
  (bind _output_get (Variables|Verse|Abilities|Slam|SetSlamDamageToCreaturesVerse (Math|Float|Clamp(Float) Damage 0.0 1e+04)))
  (ScalableFloat|SetValuewithContext (Variables|ScalableFloats|GetSF_SlamDmg_Creature) _output_get))

(fn SetSlamAttackDamageToBuildingsVerse
  (params (Damage :type Float%20%28double-precision%29))
  (bind _output_get (Variables|Verse|Abilities|Slam|SetSlamDamageToBuildingsVerse (Math|Float|Clamp(Float) Damage 0.0 1e+04)))
  (ScalableFloat|SetValuewithContext (Variables|ScalableFloats|GetSF_SlamDmg_Building) _output_get)
  (ScalableFloat|SetValuewithContext (Variables|ScalableFloats|GetSF_SlamDmg_BuildingPlayer) _output_get))

(fn SetSlamAttackDamageToGuardsVerse
  (params (Damage :type Float%20%28double-precision%29))
  (bind _output_get (Variables|Verse|Abilities|Slam|SetSlamDamageToGuardsVerse (Math|Float|Clamp(Float) Damage 0.0 1e+04)))
  (ScalableFloat|SetValuewithContext (Variables|ScalableFloats|GetSF_SlamDmg_Guard) _output_get))

(fn SetSlamAttackDamageToPlayersVerse
  (params (Damage :type Float%20%28double-precision%29))
  (bind _output_get (Variables|Verse|Abilities|Slam|SetSlamDamageToPlayersVerse (Math|Float|Clamp(Float) Damage 0.0 1e+04)))
  (ScalableFloat|SetValuewithContext (Variables|ScalableFloats|GetSF_SlamDmg_Player) _output_get))

(fn SetSlamAttackDamageToVehiclesVerse
  (params (Damage :type Float%20%28double-precision%29))
  (bind _output_get (Variables|Verse|Abilities|Slam|SetSlamDamageToVehiclesVerse (Math|Float|Clamp(Float) Damage 0.0 1e+04)))
  (ScalableFloat|SetValuewithContext (Variables|ScalableFloats|GetSF_SlamDmg_Vehicle) _output_get))

(fn SetSlamAttackDamageWildlifeVerse
  (params (Damage :type Float%20%28double-precision%29))
  (bind _output_get (Variables|Verse|Abilities|Slam|SetSlamDamageToWildlifeVerse (Math|Float|Clamp(Float) Damage 0.0 1e+04)))
  (ScalableFloat|SetValuewithContext (Variables|ScalableFloats|GetSF_SlamDmg_Wildlife) _output_get))

(fn SetSlamAttackEnabledVerse
  (params (bEnabled :type Boolean))
  (Variables|RuntimeValues|SetbCanPerformSlamAttackRuntime bEnabled))

(fn SetSlamCooldownTimeVerse
  (params (Cooldown :type Float%20%28double-precision%29))
  (Variables|RuntimeValues|SetSlamCooldownTimeRuntime (Math|Float|Clamp(Float) Cooldown 1.0 6e+01)))

(fn SetSlamInitialRadiusVerse
  (params (Radius :type Float%20%28double-precision%29))
  (bind _output_get (Variables|Verse|Abilities|Slam|SetSlamInitialRadiusVerse (Math|Float|Clamp(Float) Radius 1e+02 2e+03)))
  (CallFunction|SetAbilitySFValue (Variables|ScalableFloats|GetSF_GroundWaveInitialRadius) _output_get))

(fn SetSlamRecoveryTimeVerse
  (params (RecoveryTime :type Float%20%28double-precision%29))
  (bind _output_get (Variables|Verse|Abilities|Slam|SetSlamRecoveryTimeVerse (Math|Float|Clamp(Float) RecoveryTime 0.0 6e+01)))
  (ScalableFloat|SetValuewithContext (Variables|ScalableFloats|GetSF_SlamRecoveryTime) _output_get))

(fn SetSlamShockwaveDurationVerse
  (params (Duration :type Float%20%28double-precision%29))
  (bind _output_get (Variables|Verse|Abilities|Slam|SetSlamShockwaveDurationVerse (Math|Float|Clamp(Float) Duration 1.0 3e+01)))
  (CallFunction|SetAbilitySFValue (Variables|ScalableFloats|GetSF_GroundWaveDuration) _output_get))

(fn SetSlamShockwaveMaxRadiusVerse
  (params (Radius :type Float%20%28double-precision%29))
  (bind _output_get (Variables|Verse|Abilities|Slam|SetSlamShockwaveMaxRadiusVerse (Math|Float|Clamp(Float) Radius 1e+02 5e+03)))
  (CallFunction|SetAbilitySFValue (Variables|ScalableFloats|GetSF_GroundWaveMaxRadius) _output_get))

(fn SetSpireHidden
  (params (bNewHidden :type Boolean))
  (Networking|SwitchHasAuthority
    (:Authority
      (bind _output_get (Variables|State|SetbSpawned (not bNewHidden)))
      (if (and _output_get (Variables|UserOptions|Advanced|GetbPlayMusic))
        (CallFunction|ActivateMusic)
        (else
          (CallFunction|DeactivateMusic))))))

(fn SetUpDefaultMaterials
  (for (_array_element _array_index) (Variables|Default|GetDefaultSkelMeshMaterials)
    (Rendering|Material|SetMaterial (Variables|Default|GetSkeletalMeshComponentBudgeted) _array_index _array_element))
  (|CallParentFunction))

(fn SetUpHealthThresholds
  (Utilities|Array|Clear (Variables|Default|GetHealthThresholdPercentages))
  (LocalVariables|SetHealthThresholdIncrementPercent (/ 1.0 (Math|Conversions|ToFloat(Integer) (Variables|UserOptions|Advanced|GetNumberOfHealthSegments))))
  (for _index (range 1 (Variables|UserOptions|Advanced|GetNumberOfHealthSegments))
    (Utilities|Array|Add (Variables|Default|GetHealthThresholdPercentages) (- 1.0 (* _index (LocalVariables|GetHealthThresholdIncrementPercent))))))

(fn SetWaitForAllProjectilesToSpawnBeforeNextAttackVerse
  (params (bWait :type Boolean))
  (Variables|RuntimeValues|SetbWaitForAllProjectilesToSpawnBeforeNextAttackRuntime bWait))

(fn ShouldPerformIntroSequence
  (return (or (and (CallFunction|HasValidAbilityTarget) (Variables|State|GetbEnabled)) (Variables|STData|GetbForceSpawn))))

(fn ShowMapIcon
  (if (Math|Boolean|ANDBoolean (not (Editor|InEditor)) (not (Utilities|IsValid (Variables|MapIcon|GetMapIconActor) :!class /Script/Engine.KismetSystemLibrary)) (not (Building|IsPreviewBuildingActor)))
    (bind _spawnedactor (Fortnite|Utilities|SpawnTransientActor "/CRD_OverlordSpire/GameplayActors/CRD_OverlordSpireDevice_MinimapActor.CRD_OverlordSpireDevice_MinimapActor_C" (Transformation|GetActorTransform) "AlwaysSpawn"))
    (Variables|MapIcon|SetMapIconActor _spawnedactor)
    (MapIcon|UpdateMapIconActiveState (not (Utilities|Array|IsEmpty (Variables|Default|GetSensedTargets))))))

(fn StartStateLogic
  (AI|Logic|StartLogic (Variables|Default|GetStateTree)))

(fn StopStateLogic
  (AI|Logic|StopLogic (Variables|Default|GetStateTree) "End Minigame"))

(fn UnregisterPerception
  (Query|BPUnregisterListener (Variables|ForCreative|GetPerceptionListenerHandle))
  (Utilities|Array|Clear (Variables|InternalData|GetSensedPlayerPawns))
  (Variables|Default|SetSensedTargets (Utilities|Array|MakeArray)))

(fn UpdateCurrentTarget
  (if (and (Utilities|Array|ContainsItem (Variables|Default|GetSensedTargets) (Variables|Default|GetOverrideTarget)) (Utilities|IsValid (Variables|Default|GetOverrideTarget) :!class /Script/Engine.KismetSystemLibrary))
    (CallFunction|SetCurrentTargetPlayer (Variables|Default|GetOverrideTarget))
    (else
      (|CallParentFunction))))

(fn UpdateDataRegistryValues
  (Utilities|FlowControl|Sequence
    (:then_0
      ; Beam Settings
      (Verse|Abilities|Beam|SetBeamDurationVerse (Variables|UserOptions|Advanced|GetBeamDuration))
      (Verse|Abilities|Beam|SetBeamArcVerse (Variables|UserOptions|Advanced|GetBeamArc))
      (Verse|Abilities|Beam|SetBeamDamagePerSecondtoPlayersVerse (Variables|UserOptions|Advanced|GetBeamPlayerDamagePerSecond))
      (Verse|Abilities|Beam|SetBeamDamagePerSecondtoVehiclesVerse (Variables|UserOptions|Advanced|GetBeamVehicleDamagePerSecond))
      (Verse|Abilities|Beam|SetBeamDamagePerSecondtoBuildingsVerse (Variables|UserOptions|Advanced|GetBeamBuildingDamagePerSecond))
      (Verse|Abilities|Beam|SetBeamDamagePerSecondtoGuardsVerse (Variables|UserOptions|Advanced|GetBeamGuardDamagePerSecond))
      (Verse|Abilities|Beam|SetBeamDamagePerSecondtoWildlifeVerse (Variables|UserOptions|Advanced|GetBeamWildlifeDamagePerSecond))
      (Verse|Abilities|Beam|SetBeamDamagePerSecondtoCreaturesVerse (Variables|UserOptions|Advanced|GetBeamCreatureDamagePerSecond)))
    (:then_1
      ; Slam Attack Settings
      (Verse|Abilities|Slam|SetSlamShockwaveDurationVerse (Variables|UserOptions|Advanced|GetSlamShockwaveDuration))
      (Verse|Abilities|Slam|SetSlamShockwaveMaxRadiusVerse (* (Variables|UserOptions|Advanced|GetSlamShockwaveMaxRadius) 1e+02))
      (Verse|Abilities|Slam|SetSlamInitialRadiusVerse (* (Variables|UserOptions|Advanced|GetSlamInitialRadius) 1e+02))
      (Verse|Abilities|Slam|SetSlamRecoveryTimeVerse (Variables|UserOptions|Advanced|GetSlamRecoveryTime))
      (Verse|Abilities|Slam|SetSlamAttackDamagetoPlayersVerse (Variables|UserOptions|Advanced|GetSlamAttackPlayerDamage))
      (Verse|Abilities|Slam|SetSlamAttackDamagetoVehiclesVerse (Variables|UserOptions|Advanced|GetSlamAttackVehicleDamage))
      (Verse|Abilities|Slam|SetSlamAttackDamagetoBuildingsVerse (Variables|UserOptions|Advanced|GetSlamAttackBuildingDamage))
      (Verse|Abilities|Slam|SetSlamAttackDamageWildlifeVerse (Variables|UserOptions|Advanced|GetSlamAttackWildlifeDamage))
      (Verse|Abilities|Slam|SetSlamAttackDamageCreaturesVerse (Variables|UserOptions|Advanced|GetSlamAttackCreatureDamage))
      (Verse|Abilities|Slam|SetSlamAttackDamagetoGuardsVerse (Variables|UserOptions|Advanced|GetSlamAttackGuardDamage)))
    (:then_2
      ; Homing Projectile Settings
      (Verse|Abilities|HomProj|SetNumberOfHomingProjectilesVerse (Variables|UserOptions|Advanced|GetNumber%20of%20Homing%20Projectiles))
      (Verse|Abilities|HomProj|SetHomingProjectileMaxSpeedVerse (* (Variables|UserOptions|Advanced|GetHomingProjectileMaxSpeed) 1e+02))
      (Verse|Abilities|HomProj|SetHomingProjectileDamagetoPlayersVerse (Variables|UserOptions|Advanced|GetProjectilePlayerDamage))
      (Verse|Abilities|HomProj|SetHomingProjectileDamagetoVehiclesVerse (Variables|UserOptions|Advanced|GetProjectileVehicleDamage))
      (Verse|Abilities|HomProj|SetHomingProjectileDamagetoBuildingsVerse (Variables|UserOptions|Advanced|GetProjectileBuildingDamage))
      (Verse|Abilities|HomProj|SetHomingProjectileDamageWildlifeVerse (Variables|UserOptions|Advanced|GetProjectileWildlifeDamage))
      (Verse|Abilities|HomProj|SetHomingProjectileDamageCreaturesVerse (Variables|UserOptions|Advanced|GetProjectileCreatureDamage))
      (Verse|Abilities|HomProj|SetHomingProjectileDamagetoGuardsVerse (Variables|UserOptions|Advanced|GetProjectileGuardDamage)))
    (:then_3
      (ScalableFloat|SetValuewithContext (Variables|ScalableFloats|GetSF_WeakpointHealth) (Variables|UserOptions|Basic|GetWeakpointHealth))
      (ScalableFloat|SetValuewithContext (Variables|ScalableFloats|GetSF_NumHealthThresholds) (Variables|UserOptions|Advanced|GetNumberOfHealthSegments)))))

(fn UpdateMapIconActiveState
  (params (bIsActive :type Boolean))
  (bind _mapiconactor (Variables|MapIcon|GetMapIconActor)
    (:then
      (Default|SetMapIcon _mapiconactor (select bIsActive (Variables|MapIcon|GetMinimapIcon_Active) (Variables|MapIcon|GetMinimap%20Icon%20Inactive)))
      (Class|BPSpireMinimapActor|SetMapLegendText (Variables|MapIcon|GetMapIconActor) (Utilities|String|SelectText (Variables|MapIcon|GetMapIconText_Active) (Variables|MapIcon|GetMapIconText_Inactive) bIsActive)))
    (:else)))

(fn UpdateMusicRadius
  (Networking|CanExecuteCosmeticEvents
    (:True
      (bind _returnvalue (Utilities|InstancedStruct|MakeInstancedStruct (Utilities|Struct|MakeMusicEncounterSetDynamicVolumeRadius (+ (* (Variables|UserOptions|Basic|GetActivation%20Distance) 1e+02) 5e+02))))
      (MusicEncounter|ExecuteMusicEncounterGameplayEventByStruct (GameplayEventRouter|GetGlobalEventRouter) _returnvalue self))))

(fn UpdateVisualizationState
  (Networking|CanExecuteCosmeticEvents
    (:True
      (if (and (Utilities|IsValid (Variables|/Script/Engine.StaticMeshComponent|GetStaticMesh (Variables|Default|GetActivationRangeVisualization)) :!class /Script/Engine.KismetSystemLibrary) (or (Editor|InEditor) (not (Building|IsPreviewBuildingActor))))
        (Transformation|SetRelativeScale3D (Variables|Default|GetActivationRangeVisualization) (Math|Conversions|ToVector(Float) (/ (* (Variables|UserOptions|Basic|GetActivation%20Distance) 1e+02) (Variables|Visualization|GetVisualRangeMeshBoundsRadius))))))
    (:False
      _))
  (if (and (not (Class|CreativeCommonDeviceFunctionLibrary|IsinPlayMode self)) (Variables|UserOptions|Basic|GetShowVisualizationRange))
    (Collision|SetCollisionEnabled (Variables|Default|GetVisualizationRangeTrigger) "QueryOnly")
    (else
      (Collision|SetCollisionEnabled (Variables|Default|GetVisualizationRangeTrigger) "NoCollision")
      (CallFunction|HideRangeVisual true))))

(fn ConstructionScript
  (|CallParentFunction)
  (if (Editor|InEditor)
    (Rendering|SetVisibility (Variables|Default|GetActivationRangeVisualization) (Variables|UserOptions|Basic|GetShowVisualizationRange) false)))
